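# Editor utility that batch-creates Item Shop offer display data assets from selected textures in a season-numbered folder.
# It creates one asset per texture, or one combined asset sorted by product tag order, then name.
# Product tags come from texture path substrings, asset names from prefix replacements on the first texture.
# It duplicates an empty base asset or reuses an existing one without presentations, adds contextual presentations, then saves.
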EditorUtilityBlueprint /ItemShop/UI/BatchDACreationTool/BatchDACreator
graphs [
 AddPresentationsAndSave {nodes [
   n1 = AddPresentationsAndSave!() -> (then, InOutDA:Athena Item Shop Offer Display Data Object Reference, Textures:Array of Soft Object Path Structures) {is=title K2Node_FunctionEntry}
   > n2 = For Each Loop(Array:Array of Soft Object Path Structures=@n1.Textures) -> (Array Element:Soft Object Path Structure)
   SaveLoadedAsset(execute=@n2.Completed, AssetToSave=@n1, bOnlyIfIsDirty=false)
   n3 = LoadAsset(execute=@n2.LoopBody, AssetPath=BreakSoftObjectPath(InSoftObjectPath=@n2))
   > n4 = Cast To Texture2D(Object=@n3)
   > Add!(
    execute
    TargetArray:Array of Contextual Presentation Structures=Get ContextualPresentations!(self:Athena Item Shop Offer Display Data Object Reference=@n1) -> (ContextualPresentations:Array of Contextual Presentation Structures)
    NewItem_ProductTag:Gameplay Tag Structure=Find(
     TargetMap:Map of Soft Object Path Structures to Gameplay Tag Structures=Get TextureProductTags() -> (TextureProductTags:Map of Soft Object Path Structures to Gameplay Tag Structures)
     Key:"Soft Object Path Structure (by ref)"=@n2
    ) -> (Value:Gameplay Tag Structure)
    NewItem_RenderImage:Texture 2D Soft Object Reference=Make Soft Reference!(Input:Texture 2D Object Reference=@n4) -> (Output:Texture 2D Soft Object Reference)
    NewItem_FullTileOverrideRenderImage:Texture 2D Soft Object Reference
   ) -> (then, ReturnValue) {is=title K2Node_CallArrayFunction}
   PrintString(execute=@n4.CastFailed, InString="Script run on asset that was not a Texture2D! This shouldn't even be possible...")
  ]}
 CompareAssetData {nodes [
   n1 = CompareAssetData!() -> (then, Left:"Asset Data Structure (by ref)", Right:"Asset Data Structure (by ref)") {is=title K2Node_FunctionEntry}
   n2 = ToSoftObjectPath!(InAssetData:"Asset Data Structure (by ref)"=@n1.Right) -> (ReturnValue:Soft Object Path Structure)
   n3 = Find(TargetMap=Get TextureProductTags(), Key=@n2)
   n4 = ToSoftObjectPath(InAssetData=@n1)
   n5 = Find(TargetMap=Get TextureProductTags(), Key=@n4)
   > n6 = Branch(Condition="Equal (GameplayTag)"(A=@n5, B=@n3))
   n7 = GetSoftPtrAssetName!(Asset:Soft Object Reference=To Soft Object Reference(SoftObjectPath=@n2)) -> (ReturnValue:String)
   n8 = GetSoftPtrAssetName(Asset=To Soft Object Reference(SoftObjectPath=@n4))
   > Return Node!(execute, ReturnValue:bool="integer < integer"(A=CompareStrings!(LeftStr:String=@n8, RightStr:String=@n7) -> (ReturnValue:int=0)))
   n9 = Get OrderedProductTags() -> (OrderedProductTags:Array of Gameplay Tag Structures)
   n10 = Get OrderedProductTags()
   Return Node(
    execute=@n6.else
    ReturnValue="integer < integer"(
     A=Find Item!(TargetArray:Array of Gameplay Tag Structures=@n10, ItemToFind:"Gameplay Tag Structure (by ref)"=@n5) -> (ReturnValue)
     B=Find Item(TargetArray=@n9, ItemToFind=@n3)
    )
   )
  ]}
 CreateDAsForTextures {nodes [
   CreateDAsForTextures!() -> (then, DestinationPath:String=/OfferCatalog/NewDisplayAssets/, bCreateOneAssetPerPresentation:bool=true) {is=title K2Node_FunctionEntry}
   > n1 = EditorOnly_GetCurrentFortReleaseVersion!(execute) -> (then, ReturnValue:Fort Release Version Structure)
   > n2 = EditorOnly_GetVersionSeasonNumber!(execute, Version:"Fort Release Version Structure (by ref)"=@n1) -> (then, ReturnValue:int=0)
   n3 = "To String (Integer)"(InInt=@n2)
   > Set FinalDestinationPath(FinalDestinationPath:String=Append(A=Get DestinationPath() -> (DestinationPath:String), B=Append(A=S, B=@n3))) -> (Output_Get:String)
   > n4 = GetSelectedAssetData()
   > GenerateTextureProductTags!(execute, self, TextureData:Array of Asset Data Structures=@n4) -> (then)
   > n5 = Branch(Condition=Get bCreateOneAssetPerPresentation() -> (bCreateOneAssetPerPresentation:bool))
   > n6 = For Each Loop(Array:Array of Asset Data Structures=@n4) -> (Array Element:Asset Data Structure)
   > CreateSingleDA!(
    execute
    self
    Textures:Array of Asset Data Structures=Make Array("[0]":Asset Data Structure=@n6) -> (Array:Array of Asset Data Structures)
    Folder:String=Get FinalDestinationPath() -> (FinalDestinationPath:String)
   ) -> (then)
   SortByPredicate!(
    execute=@n5.else
    Assets:Array of Asset Data Structures=@n4
    SortingPredicate:Delegate=Create Event()
    SortOrder:EAssetRegistrySortOrder Enum=Ascending
   ) -> (then)
   > CreateSingleDA(Textures=@n4, Folder=Get FinalDestinationPath())
  ]}
 CreateSingleDA {nodes [
   n1 = CreateSingleDA!() -> (then, Textures:Array of Asset Data Structures, Folder:String) {is=title K2Node_FunctionEntry}
   > n2 = GenerateTexturePaths!(execute, self, TextureData:Array of Asset Data Structures=@n1) -> (then, TexturePaths:Array of Soft Object Path Structures)
   > n3 = Set TexturePaths(TexturePaths:Array of Soft Object Path Structures=@n2) -> (Output_Get:Array of Soft Object Path Structures)
   > n4 = GenerateAssetFileName!(
    execute
    self
    Texture:Soft Object Path Structure="Get (a copy)"(Array:Array of Soft Object Path Structures=@n3) -> (Output:Soft Object Path Structure)
   ) -> (then, DAFileName:String)
   n5 = Append!(A=Get Folder() -> (Folder:String), B=/, C:String=@n4) -> (ReturnValue)
   > n6 = DoesAssetExist(AssetPath=@n5)
   > n7 = Branch(Condition=@n6)
   > n8 = LoadAsset(AssetPath=@n5)
   > n9 = Cast To AthenaItemShopOfferDisplayData!(execute, Object=@n8) -> (then, CastFailed, AsAthena Item Shop Offer Display Data:Athena Item Shop Offer Display Data Object Reference) {is=title K2Node_DynamicCast}
   > n10 = Branch(Condition=Is Empty!(TargetArray:Array of Contextual Presentation Structures=Get ContextualPresentations(self=@n9)) -> (ReturnValue:bool))
   > AddPresentationsAndSave!(
    execute
    self
    InOutDA:Athena Item Shop Offer Display Data Object Reference=@n9
    Textures:Array of Soft Object Path Structures=Get TexturePaths() -> (TexturePaths:Array of Soft Object Path Structures)
   ) -> (then) {is=title K2Node_CallFunction}
   PrintString(execute=@n10.else, InString=The existing DA for this asset already has presentations set. No files were changed/added)
   n11 = DuplicateAsset!(
    execute=@n7.else
    SourceAssetPath:String=Get Object Path String(Object=Get BaseDa() -> (BaseDa:Athena Item Shop Offer Display Data Object Reference))
    DestinationAssetPath:String=@n5
   ) -> (then, ReturnValue:Object Reference)
   > n12 = Cast To AthenaItemShopOfferDisplayData(Object=@n11)
   > AddPresentationsAndSave(InOutDA=@n12, Textures=Get TexturePaths())
  ]}
 EventGraph {links [], nodes []}
 GenerateAssetFileName {nodes [
   n1 = GenerateAssetFileName!() -> (then, Texture:Soft Object Path Structure) {is=title K2Node_FunctionEntry}
   n2 = GetSoftPtrAssetName(Asset=To Soft Object Reference(SoftObjectPath=@n1))
   > n3 = For Loop(FirstIndex=1, LastIndex=Len(S=@n2))
   PrintString(
    execute=@n3.Completed
    InString=Append(
     A="No prefix replacement found for asset name ["
     B=@n2
     C="], only adding DAv2 prefix. Please review file name and, if applicable, modify the PrefixReplacements in the script."
    )
   )
   > Return Node!(execute, DAFileName:String=Append!(A=DAv2_, B=@n2) -> (ReturnValue))
   n4 = LeftChop(SourceString=@n2, Count=@n3)
   n5 = Find(TargetMap:Map of Strings to Strings=Get PrefixReplacements() -> (PrefixReplacements:Map of Strings to Strings), Key:"String (by ref)"=@n4) -> (Value:String)
   Branch(execute=@n3.LoopBody, Condition=@n5.ReturnValue)
   n6 = Append(A=@n5, B=Right!(SourceString:String=@n2, Count:int=@n3) -> (ReturnValue:String))
   n7 = "Equal Exactly (String)"(A=@n4, B=T_UI_Featured_Shoes_)
   n8 = EndsWith(SourceString=@n6, InSuffix=_Pair, SearchCase=CaseSensitive)
   > Return Node(DAFileName=Select(Option 0:String=@n6, Option 1:String=LeftChop(SourceString=@n6, Count=5), Index=AND Boolean(A=@n8, B=@n7)) -> (ReturnValue:String))
  ]}
 GenerateTexturePaths {nodes [
   n1 = GenerateTexturePaths!() -> (then, TextureData:Array of Asset Data Structures) {is=title K2Node_FunctionEntry}
   > n2 = For Each Loop(Array=@n1)
   Return Node!(
    execute=@n2.Completed
    TexturePaths:Array of Soft Object Path Structures=Get OutTexturePaths() -> (OutTexturePaths:Array of Soft Object Path Structures)
   )
   > Add!(
    execute
    TargetArray:Array of Soft Object Path Structures=Get OutTexturePaths()
    NewItem:"Soft Object Path Structure (by ref)"=ToSoftObjectPath(InAssetData=@n2)
   ) -> (then, ReturnValue)
  ]}
 GenerateTextureProductTags {nodes [
   GenerateTextureProductTags!() -> (then, TextureData:Array of Asset Data Structures) {is=title K2Node_FunctionEntry}
   > Clear!(execute, TargetMap:Map of Soft Object Path Structures to Gameplay Tag Structures=Get TextureProductTags()) -> (then) {is=title K2Node_CallFunction}
   > n1 = For Each Loop(Array=Get TextureData() -> (TextureData:Array of Asset Data Structures))
   n2 = ToSoftObjectPath(InAssetData=@n1)
   > Add!(
    execute
    TargetMap:Map of Soft Object Path Structures to Gameplay Tag Structures=Get TextureProductTags()
    Key:"Soft Object Path Structure (by ref)"=@n2
    Value:"Gameplay Tag Structure (by ref)"=GetProductTagForTexture!(Texture:Soft Object Path Structure=@n2) -> (ProductTag:Gameplay Tag Structure)
   ) -> (then) {is=title K2Node_CallFunction}
  ]}
 GetProductTagForTexture {nodes [
   n1 = GetProductTagForTexture!() -> (then, Texture_PathString:String) {is=title K2Node_FunctionEntry}
   > n2 = Branch(Condition=Contains(SearchIn=@n1, Substring=Juno_2dRenders))
   > Return Node!(execute, ProductTag:Gameplay Tag Structure=(TagName="Product.Juno"))
   n3 = Branch(execute=@n2.else, Condition=Contains(SearchIn=@n1, Substring=Dealership_2dRenders))
   > Return Node(ProductTag=(TagName="Product.DelMar"))
   n4 = Branch(execute=@n3.else, Condition=Contains(SearchIn=@n1, Substring=Sparks_2dRenders))
   > Return Node(ProductTag=(TagName="Product.Sparks"))
   Return Node(execute=@n4.else, ProductTag=(TagName="Product.BR"))
  ]}
]
variables [
 BaseDa {type=AthenaItemShopOfferDisplayData, default=&/EmptyNewDA}
 TextureProductTags {}
]
